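# Translucent, two-sided unlit Niagara particle material, named for the eye flames of the EthicalNoggin backpack.
# SplineThicken combines with a world position offset: a sine wiggle, upward push and clamped drag, weighted along inverted V.
# Drag compares particle position to a dynamic parameter, with Position X feeding all three axes.
# Particle color feeds emissive adjustments; panned noise, edge mask, V gradient and depth fade form the flame mask.
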Material /BRCosmetics/Accessories/FORT_Backpacks/Backpack_M_MED_EthicalNoggin/FX/Materials/M_EthicalNoggin_Eye_Flames
graph {
 properties {
  bAutomaticallySetUsageInEditor=false
  bUsedWithNiagaraMeshParticles=true
  bUsedWithNiagaraSprites=true
  blendMode=BLEND_Translucent
  shadingModel=MSM_Unlit
  twoSided=true
  usedShadingModels=Unlit
 }
 nodes [
  n1 = DynamicParameter() {defaultValue=<rgba 5 5 15 1>, paramNames [WidthBase,WidthTip,DepthFade,Param4]}
  n2 = TextureCoordinate()
  n3 = DynamicParameter() {parameterIndex=2, paramNames [Speed Y Offset, Param2, Param3, Param4]}
  n4 = VertexColor()
  n5 = Add(A=Add(A=ComponentMask(None=@n2) {g=true}, B=@n3.Speed Y Offset), B=@n4.R)
  n6 = ScalarParameter() {defaultValue=1, group=WPO, parameterName=WPO Frequency, sliderMax=2}
  n7 = Time() {period=1}
  n8 = Multiply(A=@n4.R, B=ScalarParameter() {defaultValue=1, group=WPO, parameterName=WPO Speed, sliderMax=1})
  n9 = Sine(None=Add(A=Multiply(A=@n5, B=@n6), B=Multiply(A=@n7, B=@n8)))
  n10 = VectorParameter() {group=WPO, parameterName=WPO Tangent Vector, defaultValue=<rgba 1 0 1 1>}
  n11 = Multiply(A=Multiply(A=@n9, B=Transform(None=@n10.RGB)), B=ScalarParameter() {defaultValue=1, group=WPO, parameterName=WPO Wiggle, sliderMax=10})
  n12 = Multiply(
   A=Multiply(A=ConstantBiasScale(None=@n9) {bias=0.8}, B=Constant3Vector() {constant=<rgba 0 0 1 0>})
   B=ScalarParameter() {defaultValue=1, group=WPO, parameterName=WPO Upwards, sliderMax=50}
  )
  n13 = DynamicParameter() {parameterIndex=1, paramNames [Position X, Position Y, Position Z, Param4]}
  n14 = Subtract(
   A=MaterialFunctionCall(X=@n13.Position X, Y=@n13.Position X, Z=@n13.Position X) {materialFunction=&MakeFloat3}.Result
   B=ParticlePositionWS()
  )
  n15 = ScalarParameter() {defaultValue=20, group=WPO, parameterName=WPO Max Drag, sliderMax=50}
  n16 = Multiply(A=@n15) {constB=-1}
  n17 = Add(
   A=Multiply(A=Add(A=@n11, B=@n12), B=ScalarParameter() {defaultValue=1, group=WPO, parameterName=WPO Wave, sliderMax=10, sortPriority=33})
   B=Multiply(
    A=VectorParameter() {group=WPO, parameterName=WPO Drag Direction Mask, defaultValue=<rgba 1 1 1 1>}.RGB
    B=Clamp(None=@n14, Min=@n16, Max=@n15)
   )
  )
  n18 = TextureCoordinate()
  n19 = Power(
   Base=OneMinus(None=ComponentMask(None=@n18) {g=true})
   Exp=ScalarParameter() {defaultValue=1, group=WPO, parameterName=WPO Bend, sliderMax=10}
  )
  Add(
   A=MaterialFunctionCall(WidthBase=@n1.WidthBase, WidthTip=@n1.WidthBase) {materialFunction=&SplineThicken}.WorldPositionOffset
   B=Multiply(A=ScalarParameter() {group=WPO, parameterName=WPO Intensity, sliderMax=100}, B=LinearInterpolate(B=@n17, Alpha=@n19))
  )
  MaterialFunctionCall(Emissive=Multiply(A=ParticleColor().RGB)) {materialFunction=&MF_FX_EmissiveAdjustments}
  MaterialFunctionCall(
   Brightness=ScalarParameter() {defaultValue=0.95, parameterName=Fresnel Brightness}
   Exponent=ScalarParameter() {defaultValue=16, parameterName=Fresnel Exponent}
  ) {materialFunction=&MF_Fresnel_FX}
  n20 = TextureCoordinate()
  n21 = Multiply(
   A=ComponentMask(None=@n20) {g=true}
   B=MaterialFunctionCall!(UVs=@n20, Mult=Constant() {r=4}, Exp=Constant() {r=1}) {materialFunction=/Game/Effects/Fort_Effects/MaterialFunctions/MF_SquareEdgeMask}.SquareMask
  )
  n22 = Add(
   A=MaterialFunctionCall!(UV Channel, U Freq Multiplier=Constant() {r=0.5}, V Freq Multiplier=Constant() {r=0.5}, Pan Speed U, Pan Speed V=Constant() {r=0.25}) {materialFunction=/Game/Effects/Niagara/MasterMaterials/Functions/UVs/MF_FX_TextureUVPanner}.Result
   B=@n3.Speed Y Offset
  )
  n23 = TextureSample(UVs=@n22) {samplerType=SAMPLERTYPE_Grayscale, texture=&LowResBlurredNoise}
  Multiply(A=Power(Base=Multiply(A=@n21, B=@n23.R)) {constExponent=1}, B=DepthFade(FadeDistance=@n1.DepthFade))
  TextureCoordinate()
  TextureSample(UVs=@n22) {samplerType=SAMPLERTYPE_Grayscale, texture=/Game/Effects/Fort_Effects/Textures/Fire/T_WispyFireNoise_Soft_M}
  VertexTangentWS()
 ]
}
