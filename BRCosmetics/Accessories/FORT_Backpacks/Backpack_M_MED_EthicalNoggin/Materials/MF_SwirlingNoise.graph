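# Material function for the EthicalNoggin backpack that outputs an animated swirling noise color.
# Parallax occlusion mapping on a swirl heightmap feeds polar coordinates, which build time-scrolled swirl UVs and a radial mask.
# Noise and normal samples are offset by a distortion texture, with explicit DDX/DDY from selectable UVs controlling mip choice.
# Fresnel, noise and vignette masks shape a dark-to-light color blend, finished with brightness and exponent controls.
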MaterialFunction /BRCosmetics/Accessories/FORT_Backpacks/Backpack_M_MED_EthicalNoggin/Materials/MF_SwirlingNoise
graph {nodes [
  n1 = NamedRerouteUsage()
  n2 = InverseLinearInterpolate(Value=@n1) {constA=0.45, constB=-0.5}
  n3 = Max(A=NamedRerouteUsage(), B=@n2)
  n4 = Saturate(None=Multiply(A=NamedRerouteUsage(), B=Add(A=@n3, B=@n2)))
  n5 = NamedRerouteUsage()
  n6 = InverseLinearInterpolate(
   A=@n5
   B=Add(A=@n5, B=ScalarParameter() {group=SwirlingNoise, parameterName=VirgnettingSoftness})
   Value=Length(None=MaterialFunctionCall(Float2=@n1) {materialFunction=&BreakOutFloat2Components}.R) {is=class MaterialExpressionLength}
  ) {constA=0.5}
  n7 = Saturate(None=Multiply(A=@n6, B=@n6))
  n8 = Multiply(
   A=LinearInterpolate(
    A=VectorParameter() {group=SwirlingNoise, parameterName=NoiseColorDark}.RGB
    B=VectorParameter() {group=SwirlingNoise, parameterName=NoiseColorLight, defaultValue=<rgba 0.9479 0.9479 0.9479 1>}.RGB
    Alpha=Add(A=@n4, B=@n7)
   )
   B=NamedRerouteUsage()
  )
  n9 = ScalarParameter() {defaultValue=1, group=SwirlingNoise, parameterName=GlobalBrightness}
  FunctionOutput(None=Power(Base=Multiply(A=@n8, B=@n9), Exp=ScalarParameter() {defaultValue=1, group=SwirlingNoise, parameterName=GlobalExp}))
  n10 = DotProduct(A=CameraVectorWS(), B=VertexNormalWS())
  NamedRerouteDeclaration(None=MaterialFunctionCall(
    Input=@n10
    Target Low=ScalarParameter() {group=SwirlingNoise, parameterName=VignettingMin}
    Target High=ScalarParameter() {defaultValue=0.5, group=SwirlingNoise, parameterName=VignettingMax}
   ) {materialFunction=&RemapValueRange}.Result) {name=VignettingMask, nodeColor=<rgba 0.6406 0 1 1>}
  n11 = ScalarParameter() {group=SwirlingNoise, parameterName=Swirl_UseSecondUVChannelForPanning}
  n12 = NamedRerouteDeclaration(None=LinearInterpolate(A=TextureCoordinate(), B=TextureCoordinate() {coordinateIndex=1}, Alpha=Round(None=@n11))) {name=CustomUV, nodeColor=<rgba 0.2656 1 0 1>}
  NamedRerouteDeclaration(None=DDX(Value=@n12)) {name=DDX, nodeColor=<rgba 0.1455 0.3732 1 1>}
  NamedRerouteDeclaration(None=DDY(Value=@n12)) {name=DDY, nodeColor=<rgba 0.1455 0.3732 1 1>}
  n13 = TextureObject() {texture=/BRCosmetics/Accessories/FORT_Backpacks/Backpack_M_MED_EthicalNoggin/Textures/T_EthicalNoggin_Backpack_Swirl_HeightMap}
  n14 = ScalarParameter() {defaultValue=0.3, group=SwirlingNoise, parameterName=Height_Ratio}
  n15 = Constant() {r=1}
  n16 = Constant() {r=16}
  n17 = VectorParameter() {group=SwirlingNoise, parameterName=SwirlHoleUV, defaultValue=<rgba 1 1 0 0>}
  n18 = Add(
   A=Multiply(
    A=NamedRerouteUsage()
    B=MaterialFunctionCall(X=@n17.R, Y=@n17.R) {materialFunction=&MakeFloat2}.Result
   )
   B=MaterialFunctionCall(X=@n17.B, Y=@n17.B) {materialFunction=&MakeFloat2}.Result
  )
  n19 = Constant4Vector() {constant=<rgba 1 0 0 0>}
  n20 = ScalarParameter() {defaultValue=2, group=SwirlingNoise, parameterName=RadialScale}
  n21 = StaticSwitchParameter(True=Constant() {r=-1}, False=Constant() {r=1}) {group=SwirlingNoise, parameterName=Swirling_CCW?}
  n22 = MaterialFunctionCall(Float2=MaterialFunctionCall!(
    UV Coord=MaterialFunctionCall!(
     Heightmap Texture=@n13
     Height Ratio=@n14
     Min Steps=@n15
     Max Steps=@n16
     UVs=@n18
     Heightmap Channel=@n19.RGBA
     Reference Plane
     -----------------------------
     Use World Coordinates
     Specify Manual Texture Size
     Manual Texture Size
     ------------------------------
     "Render Shadows (Occlusion Mapping)"
     Light Vector
     Shadow Steps
     Shadow Penumbra
     Transform To VertexNormal
    ) {materialFunction=/Engine/Functions/Engine_MaterialFunctions01/Texturing/ParallaxOcclusionMapping}.Parallax UVs
    LenghtScale=Constant() {r=1}
    RadialScale=Multiply(A=@n20, B=@n21)
   ) {materialFunction=/Game/Global/MaterialFunctions/TechArt/MF_PolarCoordinates}.Result) {materialFunction=&BreakOutFloat2Components}
  NamedRerouteDeclaration(None=@n22.R) {name=RadialMask, nodeColor=<rgba 1 0 0.8438 1>}
  n23 = Multiply(A=Time(), B=ScalarParameter() {group=SwirlingNoise, parameterName=SwirlingSpeed}) {constB=0.1}
  NamedRerouteDeclaration(None=MaterialFunctionCall(X=Add(A=@n23, B=@n22.R), Y=Add(A=@n22.R, B=@n22.R)) {materialFunction=&MakeFloat2}.Result) {name=SwirlingUV, nodeColor=<rgba 0 1 0.8359 1>}
  n24 = VectorParameter() {group=SwirlingNoise, parameterName=DistortionTiling, channelNames {a="", b="", g=Tile Y, r=Tile X}, defaultValue=<rgba 1 1 0 0>}
  n25 = Multiply(A=NamedRerouteUsage(), B=ComponentMask(None=@n24) {g=true, r=true})
  n26 = TextureSampleParameter2D!(UVs=@n25, "DDX(UVs)"=NamedRerouteUsage(), "DDY(UVs)"=NamedRerouteUsage(), Apply View MipBias) {
   group=SwirlingNoise
   mipValueMode=TMVM_Derivative
   parameterName=DistortionTexture
   samplerType=SAMPLERTYPE_LinearColor
   texture=&T_NoisePack_LinearColor
  }
  NamedRerouteDeclaration(None=Multiply(A=ChannelMaskParameter(None=@n26.RGB) {group=SwirlingNoise, parameterName=DistortionMaskChannel}.R) {constB=0.25}) {name=DistortionMask, nodeColor=<rgba 0 1 0.0859 1>}
  NamedRerouteDeclaration(None=ComponentMask(None=ConstantBiasScale(None=@n26.RGB) {bias=-0.5, scale=2}) {g=true, r=true}) {name=UVDistortion, nodeColor=<rgba 1 0 0.75 1>}
  n27 = NamedRerouteUsage()
  n28 = VectorParameter() {group=SwirlingNoise, parameterName=NoiseTiling, channelNames {a=Offset Y, b=Offset X, g=Tile Y, r=Tile X}, defaultValue=<rgba 2 2 0 0>}
  n29 = MaterialFunctionCall(X=@n28.Tile X, Y=@n28.Tile X) {materialFunction=&MakeFloat2}
  n30 = Add(
   A=Add(
    A=Multiply(A=@n27, B=@n29.Result)
    B=MaterialFunctionCall(X=@n28.Offset X, Y=@n28.Offset X) {materialFunction=&MakeFloat2}.Result
   )
   B=Multiply(A=ScalarParameter() {defaultValue=0.1, group=SwirlingNoise, parameterName=DistortionStrength}, B=NamedRerouteUsage())
  )
  n31 = NamedRerouteUsage()
  n32 = NamedRerouteUsage()
  n33 = TextureSampleParameter2D(UVs=@n30, "DDX(UVs)"=@n31, "DDY(UVs)"=@n32) {
   group=SwirlingNoise
   mipValueMode=TMVM_Derivative
   parameterName=NoiseTexture
   samplerType=SAMPLERTYPE_Grayscale
   texture=&LowResBlurredNoise
  }
  NamedRerouteDeclaration(None=@n33.R) {name=NoiseSample, nodeColor=<rgba 0 0.3438 1 1>}
  n34 = TextureSampleParameter2D(UVs=@n30, "DDX(UVs)"=@n31, "DDY(UVs)"=@n32) {
   group=SwirlingNoise
   mipValueMode=TMVM_Derivative
   parameterName=NoiseNormal
   samplerType=SAMPLERTYPE_Normal
   texture=&FlatNormalBlue
  }
  NamedRerouteDeclaration(None=@n34.RGB) {name=NoiseNormalSample, nodeColor=<rgba 1 0 0.3516 1>}
  n35 = Constant()
  n36 = LinearInterpolate(
   A=MaterialFunctionCall(
    Input=@n10
    Target Low=Constant() {r=-0.5}
    Target High=ScalarParameter() {defaultValue=1.5, group=SwirlingNoise, parameterName=FresnelNoiseMax}
   ) {materialFunction=&RemapValueRange}.Result
   B=MaterialFunctionCall(
    Input=@n10
    Target Low=Constant() {r=-0.5}
    Target High=ScalarParameter() {defaultValue=0.5, group=SwirlingNoise, parameterName=FresnelNoiseMin}
   ) {materialFunction=&RemapValueRange}.Result
   Alpha=NamedRerouteUsage()
  )
  n37 = ScalarParameter() {defaultValue=0.6, group=SwirlingNoise, parameterName=FresnelHardness}
  n38 = NamedRerouteUsage()
  n39 = LinearInterpolate(
   A=ScalarParameter() {defaultValue=0.5, group=SwirlingNoise, parameterName=NoiseSubtractMax}
   B=ScalarParameter() {defaultValue=0.25, group=SwirlingNoise, parameterName=NoiseSubtractMin}
   Alpha=NamedRerouteUsage()
  )
  NamedRerouteDeclaration(None=Subtract(A=SphereMask(A=@n10, B=@n35, Radius=@n36, Hardness=@n37), B=Multiply(A=@n38, B=@n39))) {name=FresnelMask, nodeColor=<rgba 1 0.7031 0 1>}
  n40 = NamedRerouteUsage()
  n41 = OneMinus(None=DotProduct(A=CameraVectorWS(), B=Transform(None=@n40)))
  NamedRerouteDeclaration(None=Power(Base=ConstantBiasScale(None=@n41))) {name=NoiseNormalFresnel, nodeColor=<rgba 0 1 0.0859 1>}
  n42 = VectorParameter() {group=SwirlingNoise, parameterName=DistortionSpeed, channelNames {a="", b="", g=Speed Y, r=Speed X}, defaultValue=<rgba 0 0 1 1>}
  Panner(Coordinate=@n25, Speed=ComponentMask(None=@n42) {g=true, r=true})
 ]}
