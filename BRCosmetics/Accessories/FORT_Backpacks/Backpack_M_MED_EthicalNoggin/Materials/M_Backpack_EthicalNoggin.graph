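# Master material for the EthicalNoggin back bling, built between the 2023 character parent PreFX and PostFX functions.
# The Tech Art Mask's blue channel replaces emissive with swirling noise and blacks out base color and specular there.
# Emissive varies by time of day, with evening and night using an HSV-adjusted version that boosts saturation and takes value from hue.
# The time-of-day intensity scalar reads only the Morning channel for all four periods.
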Material /BRCosmetics/Accessories/FORT_Backpacks/Backpack_M_MED_EthicalNoggin/Materials/M_Backpack_EthicalNoggin
graph {
 properties {
  bUseMaterialAttributes=true
  bUsedWithMeshDeformer=true
  bUsedWithMorphTargets=true
  bUsedWithNiagaraMeshParticles=true
  bUsedWithSkeletalMesh=true
  pixelDepthOffsetMode=PDOM_Legacy
  shadingModel=MSM_FromMaterialExpression
  usedShadingModels=Default Lit | Subsurface | Clear Coat | Subsurface Profile | Cloth
 }
 nodes [
  n1 = MaterialFunctionCall() {materialFunction=&MF_CharacterParent_2023_PreFX}
  n2 = GetMaterialAttributes(None=@n1.Result) {attributeGetTypes [#1, #2, #3]}
  n3 = ComponentMask(None=@n1.Tech Art Mask) {b=true}
  n4 = OneMinus(None=@n3)
  n5 = Add(
   A=Multiply(
    A=@n3
    B=MaterialFunctionCall() {materialFunction=&/MF_SwirlingNoise}.Result
   )
   B=Multiply(A=@n2.Emissive Color, B=@n4)
  ) {desc=Emissive}
  n6 = MaterialFunctionCall(Float3=RgbToHsv!(None=@n5)) {materialFunction=&BreakOutFloat3Components}
  n7 = HsvToRgb!(None=MaterialFunctionCall(X=@n6.R, Y=Multiply(A=@n6.G) {constB=1.2}, Z=@n6.R) {materialFunction=&MakeFloat3}.Result)
  n8 = MaterialFunctionCall(Color_Morning=@n5, Color_Day=@n5, Color_Evening=@n7, Color_Night=@n7) {materialFunction=&Color_Time_of_Day}
  n9 = VectorParameter() {
   group=TODFactor
   parameterName=EmissiveIntensity_of_day
   channelNames {a=Night, b=Evening, g=Day, r=Morning}
   defaultValue=<rgba 1 1 0.9 0.85>
  }
  n10 = MaterialFunctionCall(Value_Morning=@n9.Morning, Value_Day=@n9.Morning, Value_Evening=@n9.Morning, Value_Night=@n9.Morning) {materialFunction=&Scalar_Time_of_Day}
  MaterialFunctionCall(
   Material=SetMaterialAttributes!(
    MaterialAttributes=@n2.MaterialAttributes
    Base Color=Multiply(A=@n2.Base Color, B=@n4) {desc=BC}
    Specular=Multiply(A=@n2.Specular, B=@n4) {desc=Specular}
    Emissive Color=Multiply(A=@n8.Result, B=@n10.Result) {desc=Emissive with TOD Factor}
   ) {attributeSetTypes [#1, #2, #3]}
   M=@n1.M
   Pre Skinned Local Position=@n1.M
   Distance Blend=@n1.M
   Fresnel=@n1.M
  ) {materialFunction=&MF_CharacterParent_2023_PostFX}
 ]
}
