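# Physics asset for the EthicalNoggin backpack cosmetic, simulating a three-bone dangling "knocker" attached to the pack.
# One capsule covers the pack body and a small capsule covers each knocker bone.
# Constraints chain the knocker bones back to the pack.
# The middle joint locks both swings to rotate only in twist, while the outer joints set their wide angle on swing1 and narrow angles on the other axes.
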PhysicsAsset /BRCosmetics/Accessories/FORT_Backpacks/Backpack_M_MED_EthicalNoggin/Meshes/M_MED_EthicalNoggin_Pack_PhysicsAsset
bodies {
 dyn_knocker_1 [{shapeType=Capsule, length=2, radius=0.75, shapeName=dyn_knocker_1_capsule, center=<2 0 0>, rotation=<rot -90 0 0>}]
 dyn_knocker_2 [{shapeType=Capsule, length=3, radius=0.75, shapeName=dyn_knocker_2_capsule, center=<1.5 0 0>}]
 dyn_knocker_3 [{shapeType=Capsule, length=2, radius=1, shapeName=dyn_knocker_3_capsule, center=<2.5 0 0>, rotation=<rot -90 0 0>}]
 pack [{shapeType=Capsule, length=18.6621, radius=4.8379, shapeName=pack_capsule, center=<0 -5.4892 -4.9854>}]
}
constraints [
 {bone1Name=dyn_knocker_1, bone2Name=pack, swing1LimitDegrees=120, swing2LimitDegrees=1, twistLimitDegrees=1}
 {
  bone1Name=dyn_knocker_2
  bone2Name=dyn_knocker_1
  swing1LimitDegrees=2
  swing1Motion=Locked
  swing2LimitDegrees=2
  swing2Motion=Locked
  twistLimitDegrees=120
 }
 {bone1Name=dyn_knocker_3, bone2Name=dyn_knocker_2, swing1LimitDegrees=90, swing2LimitDegrees=2, twistLimitDegrees=2}
]
